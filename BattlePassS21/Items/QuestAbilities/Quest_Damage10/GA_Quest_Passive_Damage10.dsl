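; Passive quest ability for Battle Pass S21 that tracks how many distinct targets the owning player damages and reports the Damage10 quest stat.
; On activation it waits for damage-dealt events, filters them by instigator tags, and adds each target uniquely to an array.
; It prints the count and, at ten or more, sends the stat event to the quest manager.
; A task watching for a tag added to the player pawn clears the array.
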(event Ability|EventOnEndAbility
  (params (bWasCancelled :type Boolean))
  (Ability|EndAbility))

(event Ability|EventFailedToActivatePassiveAbility
  (params (ActorInfo :type Gameplay%20Ability%20Actor%20Info%20Structure))
  (Ability|EndAbility))

(event Ability|EventActivateAbility
  (bind _damaged_targets (Variables|Default|GetDamaged%20Targets))
  (bind _returnvalue (Utilities|Array|Length _damaged_targets))
  (|CallParentFunction)
  (bind (_asynctaskproxy _payload) (Ability|Tasks|WaitGameplayEvent "(TagName=\"Event.Damage.Dealt\")" 0 false false)
    (:then
      (Ability|Tasks|WaitGameplayTagAddToActor (Variables|GAVars|Data|GetOwningPlayerPawn) (Variables|Default|GetTag)
        (:Added
          (Utilities|Array|Clear (Variables|Default|GetDamaged%20Targets)))))
    (:EventReceived
      (if (GameplayTags|HasAnyTags (.InstigatorTags (Utilities|Struct|BreakGameplayEventData _payload)) (Variables|Default|GetTags))
        (Utilities|Array|AddUnique _damaged_targets (.Target (Utilities|Struct|BreakGameplayEventData _payload)))
        (Development|PrintString (Utilities|String|ToString(Integer) _returnvalue))
        (if (>= _returnvalue 10)
          (Quest|SendComplexCustomStatEvent (Variables|GAVars|Data|GetOwningPlayerQuestManager) 0 "(GameplayTags=((TagName=\"Athena.Quests.Damage10\")))" "(GameplayTags=)" 1))))))
